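; Building gameplay actor in the PistolCanary tentacle content, apparently an aim actor for NPCs, as its name indicates.
; Its only logic is the SetBPFactionTags custom event, which adds the actor to the Athena.Faction.PC.Hostile faction and writes its display name to the shipping log.
; BeginPlay, overlap, Tick and the construction script are empty, and nothing in the graph calls the custom event.
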(event EventBeginPlay)

(event Collision|EventActorBeginOverlap
  (params (OtherActor :type Actor%20Object%20Reference)))

(event EventTick
  (params (DeltaSeconds :type Float%20%28single-precision%29)))

(event Custom|SetBPFactionTags
  (bind _self self)
  (Affiliation|AddtoFaction _self "(TagName=\"Athena.Faction.PC.Hostile\")")
  (Fortnite|FortShippingLog _self (Utilities|String|Append "ADDED FACTION TAGS TO:: " (Utilities|GetDisplayName _self))))

(fn ConstructionScript)
