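; Tentacle boss gameplay actor for PistolCanary that sets health at match ready, refills it below half, and registers for aim assist.
; Damage writes impact FX to a Niagara Data Channel and bumps a replicated counter, so the attacking client feeds hit data to a post-process AnimBP for mesh deformation.
; Attacks include a socket-traced explosive shot and server-side shockwave, launching pawns and damaging builds.
; Ground laser logic is leftover, moved to another actor.
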(event EventBeginPlay
  (bind _bga_health (Variables|Default|GetBGA%20Health))
  ;[0.005158 0 0.2781 0.5] Init
  (AimAssist|RegisterAimAssistTarget 0 (Variables|Default|GetAutofireCollision))
  (Class|FortAsyncActionAuthorityMatchReady|MatchReadyAsync
    (:OnReady
      (Variables|Default|SetSaveTentacleRef (+ (Variables|Default|GetSaveTentacleRef) 1))
      (Health|SetMaxHealthDontCreateASC _bga_health)
      (Health|SetHealth _bga_health))))

(event EventOnDamageServer
  (params (Damage :type Float%20%28single-precision%29) (DamageTags :type Gameplay%20Tag%20Container%20Structure%20%28by%20ref%29) (Momentum :type Vector :default "0, 0, 0") (HitInfo :type Hit%20Result%20Structure%20%28by%20ref%29) (InstigatedBy :type Controller%20Object%20Reference) (DamageCauser :type Actor%20Object%20Reference) (EffectContext :type Gameplay%20Effect%20Context%20Handle%20Structure))
  ; Health & Damage
  (Utilities|FlowControl|Sequence
    (:then_0
      ;;[0.005158 0 0.2781 0.5] Reset Health if < 50%
      (if (< (Health|GetHealthPercent self) 0.5)
        (Health|SetHealth (Variables|Default|GetBGA%20Health)) ; HEALTH FAILSAFE
      ))
    (:then_1
      ;; Hit Impact Deformer
      (Variables|Default|SetHit%20Info HitInfo)
      (Variables|Default|SetHitInstigator InstigatedBy)
      (Variables|Default|SetSaveHit (+ (Variables|Default|GetSaveHit) 1)))))

(event EventOnDeathServer
  (params (Damage :type Float%20%28single-precision%29) (DamageTags :type Gameplay%20Tag%20Container%20Structure%20%28by%20ref%29) (Momentum :type Vector :default "0, 0, 0") (HitInfo :type Hit%20Result%20Structure%20%28by%20ref%29) (InstigatedBy :type Controller%20Object%20Reference) (DamageCauser :type Actor%20Object%20Reference) (EffectContext :type Gameplay%20Effect%20Context%20Handle%20Structure))
  ;[0.04479 0.04479 0.04479 0.5] DEBUG
  (Variables|Default|SetisDead true))

(event EventOnDamagePlayEffects
  (params (Damage :type Float%20%28single-precision%29) (DamageTags :type Gameplay%20Tag%20Container%20Structure%20%28by%20ref%29) (Momentum :type Vector :default "0, 0, 0") (HitInfo :type Hit%20Result%20Structure%20%28by%20ref%29) (InstigatedBy :type Fort%20Pawn%20Object%20Reference) (DamageCauser :type Actor%20Object%20Reference) (EffectContext :type Gameplay%20Effect%20Context%20Handle%20Structure))
  (bind (_bblockinghit _binitialoverlap _time _distance _location _impactpoint _normal _impactnormal _physmat _hitactor _hitcomponent _hitbonename _bonename _hititem _elementindex _faceindex _tracestart _traceend) (Collision|BreakHitResult HitInfo))
  ; Health & Damage
  ;;[1 0.2929 0 0.8191] Hit React FX
  (NiagaraDataChannel|WriteToNiagaraDataChannel(Legacy) :Channel "/PistolCanary/Assets/FX/NDC_PistolCanary_Tentacle_Damage.NDC_PistolCanary_Tentacle_Damage" :SearchParams (Utilities|Struct|MakeNiagaraDataChannelSearchParameters 0 _location) :bVisibleToBlueprint true :bVisibleToNiagaraCPU true :bVisibleToNiagaraGPU true :ImpactPoint _location :ImpactNormal _normal :ImpactPhysicalSurface "SurfaceType_Default" :ScaleMultiplier (Math|Float|NormalizetoRange Damage 1e+01 1.8e+02) :ImpactReflectionVector (Math|Vector|GetReflectionVector (Math|Vector|GetUnitDirection(Vector) _tracestart _traceend) _normal)))

(event Custom|Spawn_ElectricFieldVolume)

(event Custom|ToggleVisibility
  ; Toggle Visibility
  (Rendering|SetVisibility (Variables|Default|GetSkeletalMesh) false true))

(event Custom|Destroy_ElectricField(BGA_Volume))

(event Custom|SpawnHomingProjectile_WIP)

(event Custom|ForceSaveHits
  (bind _skeletalmesh (Variables|Default|GetSkeletalMesh))
  (bind _returnvalue (Components|SkeletalMesh|GetPostProcessInstance _skeletalmesh))
  (bind _hit_info (Variables|Default|GetHit%20Info))
  (bind (_bblockinghit _binitialoverlap _time _distance _location _impactpoint _normal _impactnormal _physmat _hitactor _hitcomponent _hitbonename _bonename _hititem _elementindex _faceindex _tracestart) (Collision|BreakHitResult _hit_info))
  (bind _as_abp_pistol_canary_tentacle_boss_post_process (Variables|Default|GetAs%20ABP%20Pistol%20Canary%20Tentacle%20Boss%20Post%20Process))
  ; On hit send over hit location and bone name
  (Utilities|IsValid _hitcomponent
    (:Is%20Valid
      (bind _asfort_player_controller (Utilities|Casting|CastToFortPlayerController (Variables|Default|GetHitInstigator)))
      (if (or (Utilities|Array|ContainsItem (FortPawn|GetAllLocalViewingPlayerControllers (Game|Player|GetPlayerPawn _asfort_player_controller)) _asfort_player_controller) (Utilities|Array|ContainsItem (Fortnite|GetLocalFortPlayerControllers true) _asfort_player_controller))
        (Utilities|FlowControl|Sequence
          (:then_0
            (if (!= (Variables|Default|GetAs%20ABP%20Pistol%20Canary%20Tentacle%20Boss%20Post%20Process) _returnvalue)
              (bind _asabp_pistol_canary_tentacle_boss_post_process (Utilities|Casting|CastToABP_PistolCanary_TentacleBoss_PostProcess _returnvalue))
              (Variables|Default|SetAs%20ABP%20Pistol%20Canary%20Tentacle%20Boss%20Post%20Process _asabp_pistol_canary_tentacle_boss_post_process)))
          (:then_1
            (Utilities|Array|Add (Variables|/PistolCanary/Assets/Meshes/Deform/ABP_PistolCanary_TentacleBoss_PostProcess.ABP_PistolCanary_TentacleBoss_PostProcess_C|GetHitLocationsWS _as_abp_pistol_canary_tentacle_boss_post_process) _location)
            (Utilities|Array|Add (Variables|/PistolCanary/Assets/Meshes/Deform/ABP_PistolCanary_TentacleBoss_PostProcess.ABP_PistolCanary_TentacleBoss_PostProcess_C|GetHitTraceStartsWS _as_abp_pistol_canary_tentacle_boss_post_process) _tracestart)
            (Utilities|Array|Add (Variables|/PistolCanary/Assets/Meshes/Deform/ABP_PistolCanary_TentacleBoss_PostProcess.ABP_PistolCanary_TentacleBoss_PostProcess_C|GetHit%20Bones _as_abp_pistol_canary_tentacle_boss_post_process) _hitbonename)))))))

(event Custom|Attack_Phew
  (params (OffsetAngle :type Float%20%28double-precision%29))
  (bind _phew_sourcesocket (Variables|PhewPhewPhew|GetPhew_SourceSocket))
  (bind _skeletalmesh (Variables|Default|GetSkeletalMesh))
  (bind (_returnvalue_location _returnvalue_rotation) (Math|Transform|BreakTransform (Transformation|GetBoneTransform _skeletalmesh _phew_sourcesocket "RTS_World")))
  (bind (_returnvalue_rotation_roll _returnvalue_rotation_pitch _returnvalue_rotation_yaw) (Math|Rotator|BreakRotator _returnvalue_rotation))
  (bind _returnvalue (- _returnvalue_rotation_pitch OffsetAngle))
  (bind _returnvalue_2 (Math|Rotator|MakeRotator _returnvalue_rotation_roll _returnvalue _returnvalue_rotation_yaw))
  (bind _returnvalue_3 (Math|Rotator|GetRotationXVector _returnvalue_2))
  (bind _returnvalue_4 (Math|Vector|Normalize _returnvalue_3))
  (bind _returnvalue_5 (* _returnvalue_4 5e+05))
  (bind _returnvalue_6 (+ _returnvalue_location _returnvalue_5))
  (bind _phew_hitlocation (Variables|PhewPhewPhew|GetPhew_HitLocation))
  ;[0.15 0.146 0.1455 0.5] PhewPhewPhew
  (Fortnite|FortShippingLog self (Utilities|String|Append "BGA TENTACLE:: PEWPEWPEW OFFSET INDEX IS:: " (Utilities|String|ToString(Float) OffsetAngle)))
  (bind (_outhit _returnvalue_7) (Collision|LineTraceByChannel _returnvalue_location _returnvalue_6))
  (Fortnite|FortShippingLog self (Utilities|String|Append "PHEW: Start Point " (Utilities|String|ToString(Vector) _returnvalue_location) "   End Point " (Utilities|String|ToString(Vector) _returnvalue_6) "   HitBool" (Utilities|String|ToString(Boolean) _returnvalue_7)))
  (if _returnvalue_7
    (Variables|PhewPhewPhew|SetPhew_HitActor (.HitActor (Collision|BreakHitResult _outhit)))
    (bind _output_get (Variables|PhewPhewPhew|SetPhew_HitLocation (.Location (Collision|BreakHitResult _outhit))))
    (Fortnite|FortShippingLog self (Utilities|String|Append "PHEW: Line Trace Hit Location  " (Utilities|String|ToString(Vector) _output_get)))
    (Utilities|FlowControl|Sequence
      (:then_0
        (Niagara|SpawnSystematLocation "/PistolCanary/Assets/FX/TentacleAttack/NS_TentacleAttackImpact.NS_TentacleAttackImpact" _phew_hitlocation)
        (Niagara|SpawnSystematLocation "/PistolCanary/Assets/FX/TentacleAttack/NS_TentacleAttackLingering.NS_TentacleAttackLingering" _phew_hitlocation)
        (Audio|PlaySoundatLocation "/PistolCanary_Event_Audio/Assets/Section02/tacaud1000/SC_PistolCanary_Section02_tacaud1000_GroundImpact.SC_PistolCanary_Section02_tacaud1000_GroundImpact" _phew_hitlocation))
      (:then_1
        (Utilities|FlowControl|Delay 0.2)
        (Fortnite|FortShippingLog self (Utilities|String|Append "PHEW: Tentacle Fired at " (Utilities|String|ToString(Vector) (.Location (Collision|BreakHitResult _outhit)))))
        (bind _outactors (Collision|SphereOverlapActors (Variables|PhewPhewPhew|GetPhew_HitLocation) (ScalableFloat|GetValueatLevel (Variables|PhewPhewPhew|GetPhew_ExplosionRange)) (Variables|PhewPhewPhew|GetPhew_ObjectTypes)))
        (Fortnite|FortShippingLog self (Utilities|String|Append "PHEW: Targets in Explosion Range" (Utilities|String|ToString(Integer) (Utilities|Array|Length _outactors))))
        (for _array_element _outactors
          (bind _asfort_pawn (Utilities|Casting|CastToFortPawn _array_element)
            (:then
              (GameplayEffects|ApplyGameplayEffectToTarget (Ability|GetFortAbilitySystemComponent self) (Variables|PhewPhewPhew|GetGE_Phew) (Ability|GetFortAbilitySystemComponent _asfort_pawn) 0.0 (Class|AthenaFunctionLibrary|MakeSimpleFortEffectContext self self))
              (if (GameplayTags|HasAnyMatchingGameplayTags _asfort_pawn (Variables|Default|GetHeroTagCheck))
                (CallFunction|LaunchPawn 0 (ScalableFloat|GetValueatLevel (Variables|PhewPhewPhew|GetPhew_ImpulseForward_Hero)) (ScalableFloat|GetValueatLevel (Variables|PhewPhewPhew|GetPhew_ImpulseVertical_Hero)))
                (else
                  (CallFunction|LaunchPawn _asfort_pawn (ScalableFloat|GetValueatLevel (Variables|PhewPhewPhew|GetPhew_ImpulseForward)) (ScalableFloat|GetValueatLevel (Variables|PhewPhewPhew|GetPhew_ImpulseVertical))))))
            (:CastFailed
              (bind _asbuilding_actor (Utilities|Casting|CastToBuildingActor _array_element))
              (if (BuildingActor|IsPlayerBuilt _asbuilding_actor)
                (Game|Damage|ApplyDamage _array_element (ScalableFloat|GetValueatLevel (Variables|PhewPhewPhew|GetPhew_Damage_PlayerBuild)) 0 self)
                (else
                  (Game|Damage|ApplyDamage _array_element (ScalableFloat|GetValueatLevel (Variables|PhewPhewPhew|GetPhew_EnvDamage)) 0 self))))))))))

(event Custom|PlayPhewExplosion
  ;[0.15 0.146 0.1455 0.5] PhewPhewPhew
  (Niagara|SpawnSystematLocation "/PistolCanary/Assets/FX/TentacleAttack/NS_TentacleAttackImpact.NS_TentacleAttackImpact" (Variables|PhewPhewPhew|GetPhew_HitLocation)))

(event Custom|Attack_Shockwave
  (bind _self self)
  (bind _returnvalue (Transformation|GetActorLocation _self))
  (bind _shockwave_range (Variables|Shockwave|GetShockwave_Range))
  (bind _returnvalue_2 (ScalableFloat|GetValueatLevel _shockwave_range))
  ; Shockwave
  (if (Networking|IsServer)
    (Utilities|FlowControl|Sequence
      (:then_0
        (bind _outfortpawns (Fortnite|GetFortPawnsInRange _returnvalue _returnvalue_2))
        (for _array_element _outfortpawns
          (if (GameplayTags|HasAnyMatchingGameplayTags _array_element (Variables|Default|GetHeroTagCheck))
            (CallFunction|LaunchPawn _array_element (ScalableFloat|GetValueatLevel (Variables|Shockwave|GetShockwave_ForwardVelocity_Hero)) (ScalableFloat|GetValueatLevel (Variables|Shockwave|GetShockwave_UpVelocity_Hero)) true)
            (else
              (CallFunction|LaunchPawn _array_element (ScalableFloat|GetValueatLevel (Variables|Shockwave|GetShockwave_ForwardVelocity)) (ScalableFloat|GetValueatLevel (Variables|Shockwave|GetShockwave_UpVelocity)) true)))
          (GameplayCue|ExecuteGameplayCueOnActor(Burst) _array_element (Variables|Shockwave|GetGC_ShockwaveLaunch) (Ability|GameplayCue|MakeGameplayCueParameters 0.0))))
      (:then_1
        (bind _outactors (Fortnite|GetAllActorsOfClassinRange "/Script/FortniteGame.BuildingActor" _returnvalue _returnvalue_2))
        (for _array_element_2 _outactors
          (if (BuildingActor|IsPlayerBuilt _array_element_2)
            (Game|Damage|ApplyDamage _array_element_2 (ScalableFloat|GetValueatLevel (Variables|Shockwave|GetShockwave_Damage_PlayerBuild)) 0 self)))))))

(event Custom|LaunchPawn
  (params (LaunchedPawn :type Fort%20Pawn%20Object%20Reference) (ForwardVelocity :type Float%20%28double-precision%29) (UpVelocity :type Float%20%28double-precision%29) (IgnoreNextFallingDamage :type Boolean))
  ;[0.01646 0.15 0.03905 0.5] Launch
  (bind _asfort_player_pawn (Utilities|Casting|CastToFortPlayerPawn LaunchedPawn)
    (:then
      (Components|Movement|StopMovementImmediately (Variables|/Script/Engine.Character|GetCharacterMovement _asfort_player_pawn))
      (Variables|/Script/FortniteGame.FortPawn|SetbIgnoreNextFallingDamage IgnoreNextFallingDamage _asfort_player_pawn)
      (Character|LaunchCharacter _asfort_player_pawn (CallFunction|CalculateLaunchVelocity _asfort_player_pawn ForwardVelocity UpVelocity) true true)
      (Fortnite|FortShippingLog self (Utilities|String|Append "PHEW: Launched Player" (Utilities|GetDisplayName _asfort_player_pawn))))
    (:CastFailed
      (bind _asfort_aipawn (Utilities|Casting|CastToFortAIPawn LaunchedPawn))
      (Components|Movement|StopMovementImmediately (Variables|/Script/Engine.Character|GetCharacterMovement _asfort_aipawn))
      (Variables|/Script/FortniteGame.FortPawn|SetbIgnoreNextFallingDamage IgnoreNextFallingDamage _asfort_aipawn)
      (Character|LaunchCharacter _asfort_aipawn (CallFunction|CalculateLaunchVelocity _asfort_aipawn ForwardVelocity UpVelocity) true true))))

(event Custom|ShootGroundLaser)

(event Custom|Laser_Track
  ;[0.15 0.01247 0.008101 0.5] Removed to the laser origin actor. Leave the logic here in case we want it back
  (bind (_outhit _returnvalue) (Collision|LineTraceByChannel :Start (Variables|GroundLaser|GetGroundL_SKOrigin) :End (+ (Variables|GroundLaser|GetGroundL_SKOrigin) (* (Math|Vector|Normalize (Math|Rotator|GetRotationXVector (Variables|GroundLaser|GetGroundL_SKRotation))) 1.5e+05)) :TraceChannel "TraceTypeQuery1" :bTraceComplex false :DrawDebugType "ForDuration" :bIgnoreSelf true :TraceColor "(R=1.000000,G=0.000000,B=0.000000,A=1.000000)" :TraceHitColor "(R=0.000000,G=1.000000,B=0.000000,A=1.000000)" :DrawTime 5.0))
  (Development|PrintString "Hit")
  (if _returnvalue
    (bind _output_get (Variables|GroundLaser|SetGroundL_HitLocation (.Location (Collision|BreakHitResult _outhit))))
    (CallFunction|AOE _output_get)))

(event Custom|StopLaser
  ;[0.15 0.01247 0.008101 0.5] Removed to the laser origin actor. Leave the logic here in case we want it back
  (Utilities|Time|ClearandInvalidateTimerbyHandle (Variables|GroundLaser|GetTimerHandler_Laser))
  (Utilities|Time|ClearandInvalidateTimerbyHandle (Variables|GroundLaser|GetTimerHandler_AllLaser)))

(event Custom|AOE
  (params (AOECenter :type Vector))
  (bind (_groundl_skrotation_roll _groundl_skrotation_pitch _groundl_skrotation_yaw) (Math|Rotator|BreakRotator (Variables|GetGroundL_SKRotation_Roll)))
  (bind _groundl_skrotation_roll_2 (Math|Rotator|BreakRotator (Variables|GetGroundL_SKRotation_Roll)))
  (bind (_groundl_skrotation_roll_3 _groundl_skrotation_pitch_2 _groundl_skrotation_yaw_2) (Math|Rotator|BreakRotator (Variables|GetGroundL_SKRotation_Roll)))
  ;[0.15 0.01247 0.008101 0.5] Removed to the laser origin actor. Leave the logic here in case we want it back
  (bind _outactors (Collision|SphereOverlapActors AOECenter 3e+02 (Utilities|Array|MakeArray "ObjectTypeQuery1" "ObjectTypeQuery3" "ObjectTypeQuery6" "ObjectTypeQuery2")))
  (Rendering|Debug|DrawDebugSphere "0, 0, 0" 3e+02 12 "(R=0.167654,G=0.240035,B=1.000000,A=1.000000)" 5.0 5.0)
  (for _array_element _outactors
    (bind _asfort_pawn (Utilities|Casting|CastToFortPawn _array_element))
    (GameplayEffects|ApplyGameplayEffectToTarget :self (Ability|GetFortAbilitySystemComponent self) :GameplayEffectClass "/PistolCanary/Gameplay/Effects/GE_PistolCanary_Tentacle_Laser.GE_PistolCanary_Tentacle_Laser_C" :Target (Ability|GetFortAbilitySystemComponent _asfort_pawn) :Level 0.0)
    (CallFunction|LaunchPawn _asfort_pawn (ScalableFloat|GetValueatLevel (Variables|GroundLaser|GetGroundL_ForwardVelocity)) (ScalableFloat|GetValueatLevel (Variables|GroundLaser|GetGroundL_UpVelocity)) true))
  (if (<= (Utilities|Time|GetTimerElapsedTimebyHandle (Variables|GroundLaser|GetTimerHandler_AllLaser)) 3.5)
    (Variables|GroundLaser|SetGroundL_SKRotation (Math|Rotator|MakeRotator _groundl_skrotation_roll_2))
    (elif (< (Utilities|Time|GetTimerElapsedTimebyHandle (Variables|GroundLaser|GetTimerHandler_AllLaser)) 5.0)
      (Variables|GroundLaser|SetGroundL_SKRotation (Math|Rotator|MakeRotator _groundl_skrotation_roll_3 (+ _groundl_skrotation_pitch_2 2.0) _groundl_skrotation_yaw_2))
      (else
        (Variables|GroundLaser|SetGroundL_SKRotation (Math|Rotator|MakeRotator _groundl_skrotation_roll _groundl_skrotation_pitch (+ _groundl_skrotation_yaw -1e+01)))))))

(event Custom|Attack_Shockwave3
  (bind _shockwave_range (Variables|Shockwave|GetShockwave_Range))
  (bind _returnvalue (ScalableFloat|GetValueatLevel _shockwave_range))
  (bind _self self)
  (bind _returnvalue_2 (Transformation|GetActorLocation _self))
  ; Shockwave
  (if (Networking|IsServer)
    (Utilities|FlowControl|Sequence
      (:then_0
        (bind _outfortpawns (Fortnite|GetFortPawnsInRange _returnvalue_2 _returnvalue))
        (for _array_element _outfortpawns
          (if (GameplayTags|HasAnyMatchingGameplayTags (Variables|Default|GetHeroTagCheck))
            (CallFunction|LaunchPawn 0 (ScalableFloat|GetValueatLevel (Variables|Shockwave|GetShockwave_ForwardVelocity_Hero)) (ScalableFloat|GetValueatLevel (Variables|Shockwave|GetShockwave_UpVelocity_Hero)) true)
            (else
              (CallFunction|LaunchPawn _array_element (ScalableFloat|GetValueatLevel (Variables|Shockwave|GetShockwave_ForwardVelocity)) (ScalableFloat|GetValueatLevel (Variables|Shockwave|GetShockwave_UpVelocity)) true)))
          (GameplayCue|ExecuteGameplayCueOnActor(Burst) _array_element (Variables|Shockwave|GetGC_ShockwaveLaunch) (Ability|GameplayCue|MakeGameplayCueParameters 0.0))))
      (:then_1
        (bind _outactors (Fortnite|GetAllActorsOfClassinRange "/Script/FortniteGame.BuildingActor" _returnvalue_2 _returnvalue))
        (for _array_element_2 _outactors
          (if (BuildingActor|IsPlayerBuilt _array_element_2)
            (Game|Damage|ApplyDamage _array_element_2 (ScalableFloat|GetValueatLevel (Variables|Shockwave|GetShockwave_Damage_PlayerBuild)) 0 self)))))))

(fn CalculateLaunchVelocity
  (params (LaunchedPawn :type Fort%20Pawn%20Object%20Reference) (ForwardVelocity :type Float%20%28double-precision%29) (UpVelocity :type Float%20%28double-precision%29))
  (return (+ (* (Transformation|GetActorForwardVector LaunchedPawn) ForwardVelocity) (* (Transformation|GetActorUpVector LaunchedPawn) UpVelocity))))

(fn GetTargetOwner
  (return 0))

(fn IsValidAutoFireTarget
  (params (Attacker :type Actor%20Object%20Reference))
  (return (ScalableFloat|GetValueasBool (Variables|Default|GetAllowAutoFire))))

(fn OnRep_NewHealth
  (Health|SetHealth (Variables|Default|GetBGA%20Health)))

(fn OnRep_SaveHit
  (if (not (Networking|IsServer))
    (CallFunction|ForceSaveHits)))

(fn OnRep_SaveTentacleRef
  (if (not (Networking|IsServer))
    ; We no longer save a reference to the post process anim instance here because sequencer seems to create a new one when section 2 starts. We now do the caching on first hit during ForceSaveHits event
    (Components|SkeletalMesh|SetOverridePostProcessAnimBP (Variables|Default|GetSkeletalMesh) "/PistolCanary/Assets/Meshes/Deform/ABP_PistolCanary_TentacleBoss_PostProcess.ABP_PistolCanary_TentacleBoss_PostProcess_C" true)))

(fn ShouldAlliesAutoFireOnDeployedTarget
  (return false))

(fn ShouldAutoFireOnSelfDeployedTarget
  (return false))

(fn Start_FireEyeLasers
  (for _array_element (Variables|VFX|GetEyeArray)
    (bind _returnvalue (Niagara|SpawnSystemAttached "/PistolCanary/Assets/FX/TentacleAirBeam/NS_TentacleAttackAirLaser.NS_TentacleAttackAirLaser" (Variables|Default|GetSkeletalMesh) _array_element "0, 0, 0" (Math|Transform|TransformRotation (Transformation|GetBoneTransform (Variables|Default|GetSkeletalMesh) _array_element "RTS_ParentBoneSpace"))))
    (Utilities|Array|Add (Variables|VFX|GetEyeLasers) _returnvalue)))

(fn Stop_FireEyeLasers)

(fn ConstructionScript)
