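; Ground laser origin for a tentacle attack, tracing a beam from its arrow every tick to drive Niagara and looping audio.
; Server timers gate periodic AOE pulses at the hit point and end the attack.
; Each pulse applies a GameplayEffect to pawns and launches them, damages buildings, then steps the arrow's rotation in three time-based phases, sweeping the beam.
; RepNotifies handle beam start, endpoint and shutoff cosmetics.
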(event EventTick
  (params (DeltaSeconds :type Float%20%28single-precision%29))
  (CallFunction|DrawLaser))

(event Custom|StopLaser
  (bind _timerhandler_alllaser (Variables|GroundLaser|GetTimerHandler_AllLaser))
  (bind _timerhandler_aoe (Variables|GroundLaser|GetTimerHandler_AOE))
  (if (Utilities|Time|IsValidTimerHandle _timerhandler_aoe)
    (Utilities|Time|ClearandInvalidateTimerbyHandle _timerhandler_aoe)
    (if (Utilities|Time|IsValidTimerHandle _timerhandler_alllaser)
      (Utilities|Time|ClearandInvalidateTimerbyHandle _timerhandler_alllaser)
      (Transformation|SetWorldRotation (Variables|Default|GetArrow) (Variables|Default|GetOriginalRotation) false false)
      (Variables|GroundLaser|SetVFX_Duration 0.0)
      (Variables|Default|SetShouldDoAOE true))))

(event Custom|Attack_GroundLaser
  (params (Tentacle :type BGA%20Tentacle%20Object%20Reference))
  ; Set Timers
  (Variables|Default|SetTentacle Tentacle)
  (Variables|Default|SetOriginalRotation (Transformation|GetWorldRotation (Variables|Default|GetArrow)))
  (if (Networking|IsServer)
    (bind _returnvalue (Utilities|Time|SetTimerbyEvent (EventDispatchers|CreateEvent :!event Custom|ShouldAOE) 0.2 true))
    (Variables|GroundLaser|SetTimerHandler_AOE _returnvalue)
    (bind _returnvalue_2 (Utilities|Time|SetTimerbyEvent (EventDispatchers|CreateEvent :!event Custom|StopLaser) 9.5))
    (Variables|GroundLaser|SetTimerHandler_AllLaser _returnvalue_2)
    (Variables|GroundLaser|SetcanActivate true)))

(event Custom|ShouldAOE
  (Variables|Default|SetShouldDoAOE true))

(event Custom|DrawLaser
  (bind _returnvalue_location (Math|Transform|BreakTransform (Transformation|GetRelativeTransform (Variables|Default|GetNiagara))))
  (bind (_returnvalue_location_2 _returnvalue_rotation) (Math|Transform|BreakTransform (Transformation|GetWorldTransform (Variables|Default|GetArrow))))
  ; Line Trace on Client & Server
  (bind (_outhit _returnvalue) (Collision|LineTraceByChannel :Start _returnvalue_location_2 :End (+ _returnvalue_location_2 (* (Math|Rotator|GetRotationXVector _returnvalue_rotation) 1e+05)) :TraceChannel "TraceTypeQuery1" :bTraceComplex false :DrawDebugType "None" :bIgnoreSelf true :TraceColor (Variables|Default|GetTrace%20Color) :TraceHitColor "(R=0.000000,G=1.000000,B=0.000000,A=1.000000)" :DrawTime 2.0))
  (if _returnvalue
    (bind _output_get (Variables|GroundLaser|SetGroundL_HitLocation (.Location (Collision|BreakHitResult _outhit))))
    (bind _outlocallocation (Mover/MovementBases|TransformLocationtoLocal :BasePos _returnvalue_location :WorldSpaceLocation _output_get))
    (Variables|GroundLaser|SetOut%20Local%20Location _outlocallocation)
    (Variables|Default|SetNormal (.Normal (Collision|BreakHitResult _outhit)))
    (if (Networking|IsServer)
      (if (Variables|Default|GetShouldDoAOE)
        (CallFunction|AOE (Variables|GroundLaser|GetGroundL_HitLocation))
        (Variables|Default|SetShouldDoAOE false))
      (else
        (Effects|Components|ParticleSystem|SetVectorParameter (Variables|Default|GetNiagara) "Stop Location" (Variables|GroundLaser|GetOut%20Local%20Location))
        (Transformation|SetWorldLocation (Variables|Default|GetBeamGroundAudioLoop) (Variables|GroundLaser|GetGroundL_HitLocation) false false)))))

(event Custom|LaunchPawn
  (params (LaunchedPawn :type Fort%20Pawn%20Object%20Reference) (ForwardVelocity :type Float%20%28double-precision%29) (UpVelocity :type Float%20%28double-precision%29) (IgnoreFallingDamage :type Boolean))
  ; Launch Characters
  (bind _asfort_player_pawn (Utilities|Casting|CastToFortPlayerPawn LaunchedPawn)
    (:then
      (Components|Movement|StopMovementImmediately (Variables|/Script/Engine.Character|GetCharacterMovement _asfort_player_pawn))
      (Variables|/Script/FortniteGame.FortPawn|SetbIgnoreNextFallingDamage IgnoreFallingDamage _asfort_player_pawn)
      (Character|LaunchCharacter _asfort_player_pawn (CallFunction|CalculateLaunchVelocity _asfort_player_pawn ForwardVelocity UpVelocity) true true))
    (:CastFailed
      (bind _asfort_aipawn (Utilities|Casting|CastToFortAIPawn LaunchedPawn))
      (Components|Movement|StopMovementImmediately (Variables|/Script/Engine.Character|GetCharacterMovement _asfort_aipawn))
      (Variables|/Script/FortniteGame.FortPawn|SetbIgnoreNextFallingDamage IgnoreFallingDamage _asfort_aipawn)
      (Character|LaunchCharacter _asfort_aipawn (CallFunction|CalculateLaunchVelocity _asfort_aipawn ForwardVelocity UpVelocity) true true))))

(event Custom|AOE
  (params (AOECenter :type Vector))
  ; Explosion
  (bind _outactors (Collision|SphereOverlapActors AOECenter 1e+03 (Utilities|Array|MakeArray "ObjectTypeQuery1" "ObjectTypeQuery3" "ObjectTypeQuery6" "ObjectTypeQuery2"))) ; 7/9 increased from 500
  (for _array_element _outactors
    (bind _asfort_pawn (Utilities|Casting|CastToFortPawn _array_element)
      (:then
        (GameplayEffects|ApplyGameplayEffectToTarget (Ability|GetFortAbilitySystemComponent (Variables|Default|GetTentacle)) "/PistolCanary/Gameplay/Effects/GE_PistolCanary_Tentacle_Laser.GE_PistolCanary_Tentacle_Laser_C" (Ability|GetFortAbilitySystemComponent _asfort_pawn) 0.0 (Class|AthenaFunctionLibrary|MakeSimpleFortEffectContext self self))
        (CallFunction|LaunchPawn _asfort_pawn (ScalableFloat|GetValueatLevel (Variables|GroundLaser|GetGroundL_ForwardVelocity)) (ScalableFloat|GetValueatLevel (Variables|GroundLaser|GetGroundL_UpVelocity)) true))
      (:CastFailed
        (bind _asbuilding_actor (Utilities|Casting|CastToBuildingActor _array_element))
        (if (BuildingActor|IsPlayerBuilt _asbuilding_actor)
          (Game|Damage|ApplyDamage _asbuilding_actor (ScalableFloat|GetValueatLevel (Variables|GroundLaser|GetGroundL_EnvDamage_PlayerBuild)) 0 self "/Engine/EngineDamageTypes/DmgTypeBP_Environmental.DmgTypeBP_Environmental_C")
          (else
            (Game|Damage|ApplyDamage _asbuilding_actor (ScalableFloat|GetValueatLevel (Variables|GroundLaser|GetGroundL_EnvDamage)) 0 self "/Engine/EngineDamageTypes/DmgTypeBP_Environmental.DmgTypeBP_Environmental_C"))))))
  (if (<= (Utilities|Time|GetTimerElapsedTimebyHandle (Variables|GroundLaser|GetTimerHandler_AllLaser)) 3.5)
    (Transformation|AddWorldRotation (Variables|Default|GetArrow) "0.000000,-2.000000,0.000000" false false) ; 7/9 changed from -0.5
    (elif (< (Utilities|Time|GetTimerElapsedTimebyHandle (Variables|GroundLaser|GetTimerHandler_AllLaser)) 5.0)
      (Transformation|AddWorldRotation (Variables|Default|GetArrow) "-0.500000,0.000000,0.000000" false false) ; 7/9 changed from -0.3
      (else
        (Transformation|AddWorldRotation (Variables|Default|GetArrow) "0.000000,-0.500000,0.000000" false false) ; 7/9 Changed from 0
      ))))

(fn CalculateLaunchVelocity
  (params (Pawn :type Fort%20Pawn%20Object%20Reference) (Forward :type Float%20%28double-precision%29) (Up :type Float%20%28double-precision%29))
  (return (+ (* (Transformation|GetActorForwardVector Pawn) Forward) (* (Transformation|GetActorUpVector Pawn) Up))))

(fn OnRep_canActivate
  (if (Variables|GroundLaser|GetcanActivate)
    (Networking|SwitchHasAuthority
      (:Remote
        (Components|Activation|Activate (Variables|Default|GetNiagara) false)
        (Audio|Components|Audio|FadeIn (Variables|Default|GetBeamGroundAudioLoop) 0.3 1.0 0.0 "Linear" :!class /Script/Engine.AudioComponent)))))

(fn OnRep_OutLocalLocation
  (bind _out_local_location (Variables|GroundLaser|GetOut%20Local%20Location))
  (Effects|Components|ParticleSystem|SetVectorParameter (Variables|Default|GetNiagara) "Stop Location" _out_local_location)
  (Transformation|SetWorldLocation (Variables|Default|GetBeamGroundAudioLoop) _out_local_location false false))

(fn OnRep_VFX_Duration
  (if (== (Variables|GroundLaser|GetVFX_Duration) 0)
    (Niagara|SetNiagaraVariable(Float) (Variables|Default|GetNiagara) "User.Duration" (Variables|GroundLaser|GetVFX_Duration))
    (Audio|Components|Audio|FadeOut (Variables|Default|GetBeamGroundAudioLoop) 0.3 0.0 "Linear" :!class /Script/Engine.AudioComponent)))

(fn ConstructionScript)
