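; Spawner actor for the Tentacle's Energy Whip projectile; an inline note says this logic moved to the sequencer and is kept as backup.
; BeginPlay sets canActivate, whose OnRep activates the portal Niagara effect on authority.
; SpawnPrj fires a zero-gravity projectile from the arrow toward the TargetActor entry at the Tentacle's target index, then increments that index and logs.
; The angle offset is stored but unapplied, and the line trace result is unused.
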(event EventBeginPlay
  (Variables|Default|SetcanActivate true))

(event Collision|EventActorBeginOverlap
  (params (OtherActor :type Actor%20Object%20Reference)))

(event Custom|SpawnPrj
  (params (OffsetRotation :type Float%20%28double-precision%29) (Tentacle :type BGA%20Tentacle%20Object%20Reference))
  (bind _tentacle (Variables|Default|GetTentacle))
  (bind _energywhip_targetindex (Variables|/PistolCanary/Gameplay/Tentacle/BGA_Tentacle.BGA_Tentacle_C|GetEnergyWhip_TargetIndex _tentacle))
  (bind _targetactor (Variables|Default|GetTargetActor))
  (bind _output (Utilities|Array|Get(acopy) _targetactor _energywhip_targetindex))
  (bind _returnvalue (Transformation|GetActorLocation _output))
  (bind _arrow (Variables|Default|GetArrow))
  (bind _returnvalue_2 (Transformation|GetWorldLocation _arrow))
  (bind _tentacle_2 (Variables|Default|GetTentacle))
  ;[0.15 0.00922 0.01022 0.5] The logic is moved to the sequencer. Keep it in case we want it back.
  (Variables|Default|SetTentacle Tentacle)
  (Variables|Default|SetAngleOffset OffsetRotation)
  (bind _returnvalue_3 (Animation|Utilities|GetFacingRotation _returnvalue_2 _returnvalue))
  (bind (_returnvalue_roll _returnvalue_pitch _returnvalue_yaw) (Math|Rotator|BreakRotator _returnvalue_3))
  (Collision|LineTraceByChannel _returnvalue_2 _returnvalue)
  (bind _returnvalue_4 (Projectile|SpawnProjectile :ProjectileClass "/PistolCanary/Gameplay/Tentacle/PistolCanary_Prj_EnergyWhip.PistolCanary_Prj_EnergyWhip_C" :RequestedBy Tentacle :SpawnLocation _returnvalue_2 :SpawnRotation (Math|Rotator|MakeRotator _returnvalue_roll _returnvalue_pitch _returnvalue_yaw) :SpawnDirection (Math|Rotator|MakeRotator _returnvalue_roll _returnvalue_pitch _returnvalue_yaw) :bUseDefaultPhysics false :InitialSpeed (ScalableFloat|GetValueatLevel (Variables|Default|GetInitialSpeed)) :GravityScale 0.0 :ChargePercent 0.0))
  (bind _result (Math|Integer|IncrementInt (Variables|/PistolCanary/Gameplay/Tentacle/BGA_Tentacle.BGA_Tentacle_C|GetEnergyWhip_TargetIndex _tentacle_2))
    (:%20%20
      (Variables|/PistolCanary/Gameplay/Tentacle/BGA_Tentacle.BGA_Tentacle_C|SetEnergyWhip_TargetIndex _result _tentacle_2)
      (Fortnite|FortShippingLog _returnvalue_4 (Utilities|String|Append "ENERGY WHIP SPAWNED THIS OBJECT:: " (Utilities|GetDisplayName _returnvalue_4) " & TRANSFORM IS::" (Utilities|String|ToString(Transform) (Transformation|GetActorTransform _returnvalue_4)))))))

(event EventTick
  (params (DeltaSeconds :type Float%20%28single-precision%29)))

(fn OnRep_canActivate
  (if (Variables|Default|GetcanActivate)
    (Networking|SwitchHasAuthority
      (:Authority
        (Components|Activation|Activate (Variables|Default|GetNS_TentaclePortalProjectile) false)))))

(fn ConstructionScript)
