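; Energy whip projectile for PistolCanary's tentacle attack that explodes when it stops on a blocking hit.
; On stop it stores the hit location, triggers the explosion and deactivates its portal projectile effect.
; The explosion runs its own sphere overlap, applying a gameplay effect to pawns, launching them along their own up and forward vectors, and damaging player-built structures.
; Impact Niagara and sound play at the explosion point, with radius, velocities and structure damage read from scalable floats.
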(event EventBeginPlay)

(event OnComponentBeginOverlap(CapsuleComponent)
  (params (OverlappedComponent :type Primitive%20Component%20Object%20Reference) (OtherActor :type Actor%20Object%20Reference) (OtherComp :type Primitive%20Component%20Object%20Reference) (OtherBodyIndex :type Integer) (bFromSweep :type Boolean) (SweepResult :type Hit%20Result%20Structure%20%28by%20ref%29)))

(event EventOnStop
  (params (Hit :type Hit%20Result%20Structure%20%28by%20ref%29))
  (bind (_bblockinghit _binitialoverlap _time _distance _location _impactpoint _normal _impactnormal _physmat _hitactor) (Collision|BreakHitResult Hit))
  (if _bblockinghit
    (bind _output_get (Variables|Default|SetHit%20Actor _hitactor))
    (Fortnite|FortShippingLog self (Utilities|String|Append "ENERGY WHIP ENERGY BAL HIT:: " (Utilities|GetDisplayName _output_get)))
    (Variables|Default|SetExplosionLocation _location)
    (Delivery|DoExplosion)
    (Components|Activation|Deactivate (Variables|Default|GetNS_TentaclePortalProjectile))))

(event EventOnExploded
  (params (HitActors :type Array%20of%20Actor%20Object%20References) (HitResults :type Array%20of%20Hit%20Result%20Structures))
  (Utilities|FlowControl|Sequence
    (:then_0
      (bind _outactors (Collision|SphereOverlapActors (Variables|Default|GetExplosionLocation) (ScalableFloat|GetValueatLevel (Variables|Default|GetExplosionRange)) (Utilities|Array|MakeArray "ObjectTypeQuery1" "ObjectTypeQuery2" "ObjectTypeQuery3" "ObjectTypeQuery6")))
      (for _array_element _outactors
        (bind _asfort_pawn (Utilities|Casting|CastToFortPawn _array_element)
          (:then
            (GameplayEffects|ApplyGameplayEffectToTarget (Ability|GetFortAbilitySystemComponent (Actor|GetOwner)) (Variables|Default|GetGameplay%20Effect%20Class) (Ability|GetAbilitySystemComponent _asfort_pawn) 0.0 (Class|AthenaFunctionLibrary|MakeSimpleFortEffectContext (Actor|GetOwner) (Actor|GetOwner)))
            (CallFunction|LaunchPawn _asfort_pawn))
          (:CastFailed
            (bind _asbuilding_actor (Utilities|Casting|CastToBuildingActor _array_element))
            (if (BuildingActor|IsPlayerBuilt _asbuilding_actor)
              (Game|Damage|ApplyDamage _asbuilding_actor (ScalableFloat|GetValueatLevel (Variables|Default|GetEnvDamage_PlayerBuild)) 0 (Actor|GetOwner)))))))
    (:then_1
      (CallFunction|PlayExplosionEffect))))

(event Custom|PlayExplosionEffect
  (bind _explosionlocation (Variables|Default|GetExplosionLocation))
  (Niagara|SpawnSystematLocation "/PistolCanary/Assets/FX/TentacleAttack/NS_TentacleAttackImpact.NS_TentacleAttackImpact" _explosionlocation)
  (Audio|PlaySoundatLocation "/PistolCanary_Event_Audio/Assets/Section02/tacaud6000/SC_PistolCanary_Section02_Tentacle_tac6000_PrjExplode.SC_PistolCanary_Section02_Tentacle_tac6000_PrjExplode" _explosionlocation))

(event Custom|LaunchPawn
  (params (LaunchedPawn :type Fort%20Pawn%20Object%20Reference))
  (bind _asfort_player_pawn (Utilities|Casting|CastToFortPlayerPawn LaunchedPawn)
    (:then
      (Components|Movement|StopMovementImmediately (Variables|/Script/Engine.Character|GetCharacterMovement _asfort_player_pawn))
      (CallFunction|LaunchPlayer _asfort_player_pawn))
    (:CastFailed
      (bind _asfort_aipawn (Utilities|Casting|CastToFortAIPawn LaunchedPawn))
      (Components|Movement|StopMovementImmediately (Variables|/Script/Engine.Character|GetCharacterMovement _asfort_aipawn))
      (CallFunction|LaunchAI _asfort_aipawn))))

(fn CalculateLaunchVector
  (params (LaunchedPawn :type Fort%20Pawn%20Object%20Reference))
  (return (+ (* (Transformation|GetActorUpVector LaunchedPawn) (ScalableFloat|GetValueatLevel (Variables|Default|GetVerticalVelocity))) (* (Transformation|GetActorForwardVector LaunchedPawn) (ScalableFloat|GetValueatLevel (Variables|Default|GetForwardVelocity))))))

(fn LaunchAI
  (params (AIPawn :type Fort%20Pawn%20Object%20Reference))
  (Variables|/Script/FortniteGame.FortPawn|SetbIgnoreNextFallingDamage false AIPawn)
  (Character|LaunchCharacter AIPawn (CallFunction|CalculateLaunchVector AIPawn) true true))

(fn LaunchPlayer
  (params (PlayerPawn :type Fort%20Pawn%20Object%20Reference))
  (Variables|/Script/FortniteGame.FortPawn|SetbIgnoreNextFallingDamage false PlayerPawn)
  (Character|LaunchCharacter PlayerPawn (CallFunction|CalculateLaunchVector PlayerPawn) true true))

(fn ConstructionScript)
