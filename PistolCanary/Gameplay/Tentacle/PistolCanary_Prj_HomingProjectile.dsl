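; Tentacle homing projectile for PistolCanary that explodes on a blocking hit, affecting pawns and buildings in range.
; A sphere overlap applies a gameplay effect to pawns and damages buildings, with separate values for player-built ones.
; Pawns are launched along their own up and forward vectors, ignoring the next fall damage; AI also receive an explosion effect.
; Burst sound and Niagara play, and a short downward trace places a beam-impact actor on the surface below.
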(event EventOnStop
  (params (Hit :type Hit%20Result%20Structure%20%28by%20ref%29))
  (bind (_bblockinghit _binitialoverlap _time _distance _location _impactpoint _normal _impactnormal _physmat _hitactor) (Collision|BreakHitResult Hit))
  (if _bblockinghit
    (bind _output_get (Variables|Default|SetHit%20Actor _hitactor))
    (Development|PrintString (Utilities|GetDisplayName _output_get))
    (Variables|Default|SetExplosionLocation _impactpoint)
    (Delivery|DoExplosion)
    (Components|Activation|Deactivate (Variables|Default|GetNS_TentaclePortalProjectile))))

(event EventOnExploded
  (params (HitActors :type Array%20of%20Actor%20Object%20References) (HitResults :type Array%20of%20Hit%20Result%20Structures))
  (Utilities|FlowControl|Sequence
    (:then_0
      (bind _outactors (Collision|SphereOverlapActors (Variables|Default|GetExplosionLocation) (ScalableFloat|GetValueatLevel (Variables|Default|GetExplosionRange)) (Variables|Default|GetObject%20Types)))
      (for _array_element _outactors
        (bind _asfort_pawn (Utilities|Casting|CastToFortPawn _array_element)
          (:then
            (GameplayEffects|ApplyGameplayEffectToTarget (Ability|GetFortAbilitySystemComponent (Actor|GetOwner)) (Variables|Default|GetGameplay%20Effect%20Class) (Ability|GetAbilitySystemComponent _asfort_pawn) 0.0 (Class|AthenaFunctionLibrary|MakeSimpleFortEffectContext (Actor|GetOwner) (Actor|GetOwner)))
            (CallFunction|LaunchPawn _asfort_pawn))
          (:CastFailed
            (bind _asbuilding_actor (Utilities|Casting|CastToBuildingActor _array_element))
            (if (BuildingActor|IsPlayerBuilt _asbuilding_actor)
              (Game|Damage|ApplyDamage _asbuilding_actor (ScalableFloat|GetValueatLevel (Variables|Default|GetEnvDamage_PlayerBuild)) 0 (Actor|GetOwner))
              (else
                (Game|Damage|ApplyDamage _asbuilding_actor (ScalableFloat|GetValueatLevel (Variables|Default|GetEnvDamage)) 0 (Actor|GetOwner) "/Engine/EngineDamageTypes/DmgTypeBP_Environmental.DmgTypeBP_Environmental_C")))))))
    (:then_1
      (CallFunction|PlayExplosionEffect))))

(event Custom|PlayExplosionEffect
  (bind _explosionlocation (Variables|Default|GetExplosionLocation))
  (Audio|PlaySoundatLocation (Variables|Default|GetPrjExplodeSound) _explosionlocation)
  (Niagara|SpawnSystematLocation "/PistolCanary/Assets/FX/TentaclePortalProjectiles/NS_TentaclePortalBurst.NS_TentaclePortalBurst" _explosionlocation)
  (bind _outhit (Collision|LineTraceByChannel _explosionlocation (+ _explosionlocation "0.000000,0.000000,-300.000000")))
  (if (Collision|BreakHitResult _outhit)
    (Game|SpawnActorfromClass "/PistolCanary/Assets/FX/TentacleCircularBeamAttack/BP_PC_Tentacle_Beam_Impact.BP_PC_Tentacle_Beam_Impact_C" (Math|Transform|MakeTransform (.Location (Collision|BreakHitResult _outhit)) (Math|Conversions|RotationFromXVector (.Normal (Collision|BreakHitResult _outhit)))) "Undefined" "MultiplyWithRoot" 0 0)))

(event Custom|LaunchPawn
  (params (LaunchedPawn :type Fort%20Pawn%20Object%20Reference))
  (bind _asfort_player_pawn (Utilities|Casting|CastToFortPlayerPawn LaunchedPawn)
    (:then
      (Components|Movement|StopMovementImmediately (Variables|/Script/Engine.Character|GetCharacterMovement _asfort_player_pawn))
      (CallFunction|LaunchPlayer _asfort_player_pawn))
    (:CastFailed
      (bind _asfort_aipawn (Utilities|Casting|CastToFortAIPawn LaunchedPawn))
      (Components|Movement|StopMovementImmediately (Variables|/Script/Engine.Character|GetCharacterMovement _asfort_aipawn))
      (CallFunction|LaunchAI _asfort_aipawn))))

(fn CalculateLaunchVector
  (params (LaunchedPawn :type Fort%20Pawn%20Object%20Reference))
  (return (+ (* (Transformation|GetActorUpVector LaunchedPawn) (ScalableFloat|GetValueatLevel (Variables|Default|GetVerticalVelocity))) (* (Transformation|GetActorForwardVector LaunchedPawn) (ScalableFloat|GetValueatLevel (Variables|Default|GetForwardVelocity))))))

(fn LaunchAI
  (params (AIPawn :type Fort%20Pawn%20Object%20Reference))
  (bind _mesh (Variables|/Script/Engine.Character|GetMesh AIPawn))
  (Variables|/Script/FortniteGame.FortPawn|SetbIgnoreNextFallingDamage true AIPawn)
  (Utilities|FlowControl|Sequence
    (:then_0
      (Character|LaunchCharacter AIPawn (CallFunction|CalculateLaunchVector AIPawn) true true)
      (GameplayEffects|ApplyGameplayEffectToTarget :self (Ability|GetAbilitySystemComponent AIPawn) :GameplayEffectClass "/PistolCanary/Gameplay/Effects/GE_PistolCanary_Tentacle_Explosion.GE_PistolCanary_Tentacle_Explosion_C" :Target (Ability|GetAbilitySystemComponent AIPawn) :Level 0.0))
    (:then_1
      (Effects|SpawnFXSystemAttached :FXSystemAsset (Variables|Default|GetFXSystem%20Asset) :AttachToComponent _mesh :AttachPointName "root" :Scale "1.000000,1.000000,1.000000" :LocationType "SnapToTarget" :bAutoDestroy true :PoolingMethod "None" :bAutoActivate true)
      (Audio|SpawnSoundAttached "/Game/Sounds/Fort_GamePlay_Sounds/Explosives/ShockwaveGrenade/SC_Grenade_Shockwave_Launch_Impulse.SC_Grenade_Shockwave_Launch_Impulse" _mesh))))

(fn LaunchPlayer
  (params (PlayerPawn :type Fort%20Pawn%20Object%20Reference))
  (Variables|/Script/FortniteGame.FortPawn|SetbIgnoreNextFallingDamage true PlayerPawn)
  (Utilities|FlowControl|Sequence
    (:then_0
      (Character|LaunchCharacter PlayerPawn (CallFunction|CalculateLaunchVector PlayerPawn) true true))
    (:then_1
      (Effects|SpawnFXSystemAttached (Variables|Default|GetFXSystem%20Asset) (Variables|/Script/Engine.Character|GetMesh PlayerPawn) "root"))))

(fn ConstructionScript)
