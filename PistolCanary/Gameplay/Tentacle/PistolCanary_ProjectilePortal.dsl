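; Projectile portal actor for the tentacle that fires PistolCanary homing projectiles toward nearby players when SpawnPrj is called.
; SpawnPrj stores the tentacle and gathers player pawns with an oriented box overlap ahead of its arrow.
; It calls SpawnOneProj up to three times by player count, each aiming at a random player, then removing them from the list.
; With no players, it fires one shot along the arrow with gravity, requested by itself rather than the tentacle.
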(event EventBeginPlay)

(event Collision|EventActorBeginOverlap
  (params (OtherActor :type Actor%20Object%20Reference)))

(event Custom|SpawnPrj
  (params (Tentacle :type BGA%20Tentacle%20Object%20Reference))
  (bind (_returnvalue_location _returnvalue_rotation) (Math|Transform|BreakTransform (Transformation|GetWorldTransform (Variables|Default|GetArrow))))
  (bind (_returnvalue_x _returnvalue_y _returnvalue_z) (Math|Vector|BreakVector (Transformation|GetWorldLocation (Variables|Default|GetArrow))))
  (Variables|Default|SetTentacle Tentacle)
  (bind _outactors (Collision|BoxOverlapActorsWithOrientation (+ (Math|Vector|MakeVector _returnvalue_x _returnvalue_y _returnvalue_z) (* (Math|Vector|Normalize (Transformation|GetForwardVector (Variables|Default|GetArrow))) (/ (Variables|Default|GetBox_Width) 2.0))) (Math|Vector|MakeVector (Variables|Default|GetBox_Width) (Variables|Default|GetBox_Length) _returnvalue_z) (Math|Rotator|MakeRotfromX (Transformation|GetForwardVector (Variables|Default|GetArrow))) (Variables|Default|GetObject%20Types) "/Script/FortniteGame.FortPlayerPawnAthena"))
  (Variables|Default|SetPlayersInRange _outactors)
  (Fortnite|FortShippingLog self (Utilities|String|Append "HOMING PROJ:  Players in Spawner Range" (Utilities|String|ToString(Integer) (Utilities|Array|Length _outactors))))
  (switch int (Utilities|Array|Length _outactors)
    (:Default
      (CallFunction|SpawnOneProj)
      (CallFunction|SpawnOneProj)
      (CallFunction|SpawnOneProj))
    (:0
      (Projectile|SpawnProjectile :ProjectileClass "/PistolCanary/Gameplay/Tentacle/PistolCanary_Prj_HomingProjectile.PistolCanary_Prj_HomingProjectile_C" :RequestedBy self :SpawnLocation _returnvalue_location :SpawnRotation _returnvalue_rotation :SpawnDirection _returnvalue_rotation :bUseDefaultPhysics false :InitialSpeed (ScalableFloat|GetValueatLevel (Variables|Default|GetInitialSpeed)) :GravityScale 0.3 :ChargePercent 0.0))
    (:1
      (CallFunction|SpawnOneProj))
    (:2
      (CallFunction|SpawnOneProj)
      (CallFunction|SpawnOneProj))))

(event EventTick
  (params (DeltaSeconds :type Float%20%28single-precision%29)))

(fn SpawnOneProj
  (bind _arrow (Variables|Default|GetArrow))
  (bind _returnvalue (Transformation|GetWorldLocation _arrow))
  (bind (_returnvalue_location _returnvalue_rotation) (Math|Transform|BreakTransform (Transformation|GetWorldTransform (Variables|Default|GetArrow))))
  (bind _element (ArrayMacros|GetRandomArrayElement (Variables|Default|GetPlayersInRange))
    (:%20%20
      (bind _returnvalue_2 (Animation|Utilities|GetFacingRotation _returnvalue (Transformation|GetActorLocation _element)))
      (Projectile|SpawnProjectile :ProjectileClass "/PistolCanary/Gameplay/Tentacle/PistolCanary_Prj_HomingProjectile.PistolCanary_Prj_HomingProjectile_C" :RequestedBy (Variables|Default|GetTentacle) :SpawnLocation _returnvalue :SpawnRotation _returnvalue_2 :SpawnDirection _returnvalue_2 :bUseDefaultPhysics false :InitialSpeed (ScalableFloat|GetValueatLevel (Variables|Default|GetInitialSpeed)) :GravityScale 0.0 :ChargePercent 0.0)
      (Utilities|Array|RemoveItem (Variables|Default|GetPlayersInRange) _element))
    (:Empty%20Array
      (Projectile|SpawnProjectile :ProjectileClass "/PistolCanary/Gameplay/Tentacle/PistolCanary_Prj_HomingProjectile.PistolCanary_Prj_HomingProjectile_C" :RequestedBy self :SpawnLocation _returnvalue_location :SpawnRotation _returnvalue_rotation :SpawnDirection _returnvalue_rotation :bUseDefaultPhysics false :InitialSpeed (ScalableFloat|GetValueatLevel (Variables|Default|GetInitialSpeed)) :GravityScale 0.3 :ChargePercent 0.0))))

(fn ConstructionScript)
